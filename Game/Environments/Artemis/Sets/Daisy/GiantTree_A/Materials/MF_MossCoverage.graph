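# Material function that layers vertex-painted Frontier moss over incoming material attributes, switched by EnableMossPainting.
# World-aligned moss textures drive specular, roughness, a flattened normal, a height mask and curve-atlas color detail.
# A Fresnel fuzz term shifts tinted base color from darkened core to brighter edge.
# HeightLerp of red vertex color against inverted height sets the blend; the TrunkHeight and edge-color gradient branches reach no output.
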MaterialFunction /Game/Environments/Artemis/Sets/Daisy/GiantTree_A/Materials/MF_MossCoverage
graph {nodes [
  Constant()
  n1 = GetMaterialAttributes(None=FunctionInput(Preview=MakeMaterialAttributes!(
     BaseColor
     Metallic
     Specular
     Roughness
     Anisotropy
     EmissiveColor
     Opacity
     OpacityMask
     Normal
     Tangent
     WorldPositionOffset
     SubsurfaceColor
     ClearCoat
     ClearCoatRoughness
     AmbientOcclusion
     Refraction
     CustomizedUVs_0
     CustomizedUVs_1
     CustomizedUVs_2
     CustomizedUVs_3
     CustomizedUVs_4
     CustomizedUVs_5
     CustomizedUVs_6
     CustomizedUVs_7
     PixelDepthOffset
     ShadingModel
     Displacement
    )) {bUsePreviewValueAsDefault=true, inputName=Material, inputType=FunctionInput_MaterialAttributes}) {attributeGetTypes [#1]}
  n2 = Constant() {r=0.01}
  n3 = NamedRerouteUsage()
  n4 = VectorParameter() {parameterName=MossEdgeColor, defaultValue=<rgba 0.875 0.5013 0.5947 1>}
  n5 = Saturate(None=Add(A=Multiply(A=@n2, B=@n3), B=Multiply(A=@n3, B=@n4.RGB)))
  n6 = ScalarParameter() {defaultValue=0.25, group=Moss, parameterName=Moss_CoreDarkening}
  n7 = ScalarParameter() {defaultValue=1, group=Moss, parameterName=Moss_FuzzyEdgeBrightness}
  n8 = TextureObjectParameter() {
   group=Top
   parameterName=Normals Top
   samplerType=SAMPLERTYPE_Normal
   sortPriority=6
   texture=/Game/Environments/Artemis/Sets/Terrain/Textures/T_Artemis_Frontier_Moss_01_N
  }
  n9 = ScalarParameter() {defaultValue=500, group=Moss, parameterName=Scale Top, sortPriority=3}
  n10 = ScalarParameter() {group=FrontierMoss, parameterName=ProjectionTransitionContrast}
  n11 = MaterialFunctionCall(TextureObject=@n8, TextureSize=@n9, WorldSpace=StaticBool(), Normal=VertexNormalWS(), ProjectionTransitionContrast=@n10) {materialFunction=&WorldAlignedNormal}
  n12 = ComponentMask(None=VertexNormalWS()) {b=true}
  n13 = MaterialFunctionCall(
   Normal=LinearInterpolate(A=@n11.XYZFlatTop, B=@n11.XYZFlatTop, Alpha=Clamp(None=@n12))
   Flatness=ScalarParameter() {group=Moss, parameterName=MossNormalFlatten}
  ) {materialFunction=&FlattenNormal}
  n14 = Power(
   Base=Fresnel(ExponentIn=ScalarParameter() {defaultValue=2, group=Moss, parameterName=FuzzExponent}, Normal=Transform(None=@n13.Result))
   Exp=ScalarParameter() {defaultValue=2, group=Moss, parameterName=FuzzEdgeClamp}
  )
  n15 = StaticBool() {value=true}
  n16 = TextureObjectParameter() {
   group=Top
   parameterName=Spec Top
   samplerType=SAMPLERTYPE_Masks
   sortPriority=4
   texture=/Game/Environments/Artemis/Sets/Terrain/Textures/T_Artemis_Frontier_Moss_01_S
  }
  n17 = WorldPosition()
  n18 = MaterialFunctionCall(Export Float 4=@n15, TextureObject=@n16, TextureSize=@n9, WorldPosition=@n17.XYZ, ProjectionTransitionContrast=@n10) {materialFunction=/Game/Weapons/WeaponSkins/Wraps/MaterialFunctions/WorldAlignedTexture}
  n19 = SetMaterialAttributes!(
   MaterialAttributes=@n1.MaterialAttributes
   BaseColor=LinearInterpolate(A=Multiply(A=@n5, B=@n6), B=Multiply(A=@n5, B=@n7), Alpha=Clamp(None=@n14))
   Specular=LinearInterpolate(
    A=ScalarParameter() {defaultValue=0.1, group=Moss, parameterName=Moss_Min_Specular}
    B=ScalarParameter() {defaultValue=0.5, group=Moss, parameterName=Moss_Max_Specular}
    Alpha=ComponentMask(None=@n18.XYZ Texture) {r=true}
   )
   Roughness=LinearInterpolate(
    A=ScalarParameter() {defaultValue=0.75, group=Moss, parameterName=Moss_Min_Roughness}
    B=ScalarParameter() {defaultValue=1, group=Moss, parameterName=Moss_Max_Roughness}
    Alpha=ComponentMask(None=@n18.XYZ Texture) {b=true}
   )
   Normal=@n13.Result
   EmissiveColor=Constant3Vector()
  ) {attributeSetTypes [
    #2
    #3
    #4
    #1
    #5
   ]}
  n20 = VertexColor()
  n21 = ScalarParameter() {defaultValue=1, group=Moss, parameterName=MossEdgeFalloff}
  n22 = MaterialFunctionCall(Transition Phase=@n20.R, Height Texture=OneMinus(None=NamedRerouteUsage()), Contrast=@n21) {materialFunction=&HeightLerp}
  FunctionOutput(None=StaticSwitchParameter(True=BlendMaterialAttributes(A=@n1.MaterialAttributes, B=@n19, Alpha=@n22.Alpha), False=@n1.MaterialAttributes) {parameterName=EnableMossPainting})
  n23 = ScalarParameter() {group=Moss, parameterName=MossColorEdge_1}
  n24 = ScalarParameter() {defaultValue=2, group=Moss, parameterName=MossColorEdge_2}
  LinearInterpolate(A=@n4.RGB, B=@n3, Alpha=Saturate(None=LinearInterpolate(A=@n23, B=@n24, Alpha=@n22.Lerp Alpha No Contrast)))
  n25 = FunctionInput(Preview=Constant()) {bUsePreviewValueAsDefault=true, inputName=TrunkHeight, inputType=FunctionInput_Scalar}
  n26 = Subtract(A=Divide(A=Subtract(A=@n20.R, B=@n25), B=@n21), B=Clamp(None=Subtract(A=@n20.R)))
  MaterialFunctionCall(In=Clamp(None=Add(A=@n26) {constB=0.2}), Contrast=ScalarParameter() {defaultValue=1, group=Moss, parameterName=MossEdgeContrast}) {materialFunction=&CheapContrast}
  n27 = TextureObjectParameter() {
   group=Top
   parameterName=Diffuse Top
   samplerType=SAMPLERTYPE_Masks
   sortPriority=4
   texture=/Game/Environments/Artemis/Sets/Terrain/Textures/T_Artemis_Frontier_Moss_01_D
  }
  n28 = MaterialFunctionCall(Export Float 4=@n15, TextureObject=@n27, TextureSize=@n9, WorldPosition=@n17.XYZ, ProjectionTransitionContrast=@n10) {materialFunction=&WorldAlignedTexture}
  NamedRerouteDeclaration(None=ComponentMask(None=@n28.XYZ Texture) {a=true}) {name=MossHeight, nodeColor=<rgba 0 0.8594 1 1>}
  n29 = LinearInterpolate(
   A=VectorParameter() {group=FrontierMoss, parameterName=Moss_Color_01, defaultValue=<rgba 0.8398 0.3833 1.0666 1>}.RGB
   B=VectorParameter() {group=FrontierMoss, parameterName=Moss_Color_02, defaultValue=<rgba 0.6574 0.3958 0.6666 1>}.RGB
  )
  n30 = VectorParameter() {group=FrontierMoss, parameterName=Moss_Color_03, defaultValue=<rgba 0.8398 0.3833 1.0666 1>}
  NamedRerouteDeclaration(None=LinearInterpolate(
    A=LinearInterpolate(A=@n29, B=@n30.RGB)
    B=VectorParameter() {group=FrontierMoss, parameterName=Moss_Color_04, defaultValue=<rgba 0.6574 0.3958 0.6666 1>}.RGB
   )) {name=MossMicroColors, nodeColor=<rgba 1 0.5859 0 1>}
  n31 = CurveAtlasRowParameter(CurveTime=LinearInterpolate(A=ScalarParameter() {defaultValue=0.15, parameterName=MossMinColVal}, Alpha=ComponentMask(None=@n28.XYZ Texture) {r=true}) {constA=0.25}) {
   atlas=/Game/Athena/Apollo/Environments/Landscape/Materials/Frontier/FrontierColorAtlas
   curve=/Game/Athena/Apollo/Environments/Landscape/Materials/Frontier/FrontierMossDetailColor
   parameterName=MossColorCurve
  }
  NamedRerouteDeclaration(None=@n31) {name=MossBaseColorDetail, nodeColor=<rgba 1 0 0.6562 1>}
  Reroute()
  SetMaterialAttributes!(MaterialAttributes=@n1.MaterialAttributes)
  Transform(None=Constant3Vector() {constant=<rgba 0 0 1 0>}) {transformSourceType=TRANSFORMSOURCE_World, transformType=TRANSFORM_Tangent}
 ]}
