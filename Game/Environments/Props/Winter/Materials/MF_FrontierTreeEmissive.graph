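# Material function for winter frontier tree props that layers a pulsing, colored glow onto input Material Attributes when UseFrontierGlow is enabled.
# The mask comes from an emissive texture, optionally combined with AO and noise-broken vertex paint, and also blends base color toward a dimmed glow color.
# A time-driven sine along second-UV V modulates color and strength.
# Hashed-speed panning streak, fresnel and bounds-gradient branches are unconnected to the output.
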MaterialFunction /Game/Environments/Props/Winter/Materials/MF_FrontierTreeEmissive
graph {nodes [
  n1 = Multiply(
   A=LinearInterpolate(A=TextureCoordinate(), B=TextureCoordinate() {coordinateIndex=1}, Alpha=ScalarParameter() {group=FrontierGlow, parameterName=UseUV2})
   B=AppendVector(
    A=ScalarParameter() {defaultValue=6, group=FrontierGlow, parameterName=Streaks_X_Tiling}
    B=ScalarParameter() {defaultValue=2, group=FrontierGlow, parameterName=Streaks_Y_Tiling}
   )
  )
  n2 = MaterialFunctionCall(Float2=Multiply(A=@n1, B=ScalarParameter() {defaultValue=7, group=FrontierGlow, parameterName=SpeedMaskTiling}) {constB=7}) {materialFunction=&BreakOutFloat2Components}
  n3 = Constant() {r=1}
  n4 = MaterialFunctionCall(
   Vec3=ObjectPositionWS()
   Vec2=Ceil(None=MaterialFunctionCall(X=@n2.R, Y=@n3) {materialFunction=&MakeFloat2}.Result)
  ) {materialFunction=&MF_Hash}
  Clamp(None=ConstantBiasScale(None=ChannelMaskParameter(None=@n4.Vec3) {parameterName=Param_1}.R))
  n5 = GetMaterialAttributes(None=FunctionInput() {inputName=Material, inputType=FunctionInput_MaterialAttributes}) {attributeGetTypes [#1, #2]}
  n6 = VertexColor()
  n7 = TextureCoordinate() {coordinateIndex=1}
  n8 = TextureSampleParameter2D(UVs=Multiply(A=@n7, B=ScalarParameter() {defaultValue=4, parameterName=VertexPaintNoiseTiling})) {
   group=FrontierGlow
   parameterName=VertexColorBlendNoise
   samplerType=SAMPLERTYPE_LinearGrayscale
   texture=/Game/Effects/Fort_Effects/Textures/Tile/T_Diagonal_Blur_Noise_M
  }
  n9 = ScalarParameter() {group=FrontierGlow, parameterName=VertexPaintContrast}
  n10 = MaterialFunctionCall(Transition Phase=@n6.G, Height Texture=@n8.R, Contrast=@n9) {materialFunction=&HeightLerp}
  n11 = FunctionInput(Preview=Constant()) {bUsePreviewValueAsDefault=true, inputName=AO, inputType=FunctionInput_Scalar}
  n12 = Power(
   Base=Multiply(A=OneMinus(None=@n11), B=ScalarParameter() {defaultValue=4, parameterName=AOBoost}) {constB=2}
   Exp=ScalarParameter() {defaultValue=4, parameterName=AOPower}
  ) {constExponent=3, desc=1}
  n13 = TextureCoordinate()
  n14 = TextureSampleParameter2D(UVs=Multiply(
    A=StaticSwitchParameter(True=@n13, False=@n7) {group=FrontierGlow, parameterName=UseUV1Emissive}
    B=ScalarParameter() {defaultValue=1, group=FrontierGlow, parameterName=GlowFXMaskTiling}
   )) {
   group=FrontierGlow
   parameterName=FroniterEmissiveTexture
   texture=/Game/Environments/Artemis/Sets/Daisy/GiantTree_A/Textures/T_Daisy_GiantTree_Bark_01_E
  }
  n15 = Multiply(
   A=StaticSwitchParameter(True=Clamp(None=@n12), False=Constant() {r=1}) {defaultValue=true, group=FrontierGlow, parameterName=AOEmissiveMask}
   B=@n14.G
  )
  n16 = StaticSwitchParameter(
   True=Multiply(
    A=MaterialFunctionCall(Transition Phase=@n6.B, Height Texture=@n8.R, Contrast=@n9) {materialFunction=&HeightLerp}.Alpha
    B=Max(A=@n10.Alpha, B=@n15)
   )
   False=@n14.G
  ) {group=FrontierGlow, parameterName=EnableEmissivePaint}
  n17 = VectorParameter() {parameterName=StreaksColor, defaultValue=<rgba 0 0.394 1 1>}
  n18 = ChannelMaskParameter(None=TextureCoordinate() {coordinateIndex=1}) {maskChannel=Green, parameterName=Param_2, defaultValue=<rgba 0 1 0 0>}
  n19 = Multiply(A=Time()) {constB=0.2}
  n20 = ConstantBiasScale(None=Sine(None=Add(A=@n18.R, B=@n19)))
  n21 = Multiply(A=@n16, B=LinearInterpolate(A=@n17.RGB, Alpha=@n20) {constB=0})
  n22 = LinearInterpolate(
   A=ScalarParameter() {defaultValue=2, group=FrontierGlow, parameterName=GlowStrength}
   B=ScalarParameter() {defaultValue=0.5, group=FrontierGlow, parameterName=GlowStrengthMin}
   Alpha=@n20
  ) {constB=0}
  n23 = Add(A=@n5.EmissiveColor, B=Power(Base=Multiply(A=@n21, B=@n22)))
  n24 = LinearInterpolate(
   A=@n5.BaseColor
   B=Multiply(A=@n17.RGB, B=ScalarParameter() {defaultValue=0.1, parameterName=DarkenEmissiveInterion}) {constB=0.25}
   Alpha=@n16
  )
  FunctionOutput(None=StaticSwitchParameter(
    True=SetMaterialAttributes!(MaterialAttributes=@n5.MaterialAttributes, EmissiveColor=@n23, BaseColor=@n24) {attributeSetTypes [#1, #2]}
    False=@n5.MaterialAttributes
   ) {parameterName=UseFrontierGlow})
  n25 = Transform(None=PixelNormalWS())
  Multiply(A=@n22, B=Power(Base=Fresnel(Normal=@n25)))
  n26 = VertexInterpolator(VS=PreSkinnedPosition())
  n27 = MaterialFunctionCall() {materialFunction=&ObjectLocalBounds}
  Multiply(A=@n20, B=ChannelMaskParameter(None=Divide(A=@n26.PS, B=@n27.Local Bounds Size)) {maskChannel=Blue, defaultValue=<rgba 0 0 1 0>}.R)
  n28 = MaterialFunctionCall(Float2=@n1) {materialFunction=&BreakOutFloat2Components}
  n29 = Time()
  n30 = ConstantBiasScale(None=ComponentMask(None=@n4.Vec2) {r=true})
  n31 = LinearInterpolate(
   A=ScalarParameter() {defaultValue=0.02, group=FrontierGlow, parameterName=MinPanSpeed}
   B=ScalarParameter() {defaultValue=0.25, group=FrontierGlow, parameterName=MaxPanSpeed}
   Alpha=Clamp(None=Power(Base=@n30))
  )
  n32 = MaterialFunctionCall(X=@n28.R, Y=Add(A=@n28.G, B=Multiply(A=@n29, B=@n31))) {materialFunction=&MakeFloat2}
  TextureSample(UVs=@n32.Result) {texture=/Game/Characters/Player/Male/Medium/Bodies/M_MED_Viper/Textures/Glitch/T_SoftNoise}
  TextureSample(UVs=@n32.Result) {samplerType=SAMPLERTYPE_Masks, texture=/Game/Environments/Apollo/Foliage/Textures_Materials/T_FrontierTree_FX}
 ]}
