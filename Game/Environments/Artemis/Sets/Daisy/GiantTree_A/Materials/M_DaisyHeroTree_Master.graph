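# Master material for the Daisy giant hero tree: masked subsurface, Nanite-ready, with IsTrunk and opaque-canopy switches.
# Tiled bark detail layers over baked normal, AO and curvature by quality level, with ground color blending and optional landscape-snapping WPO.
# Material-attribute stages add snow, emissive, moss, vertex AO, hit wobble, SH-driven fresnel and a time-of-day glow sphere.
# Trunks blend into landscape runtime virtual textures at Epic quality, before Curie and Creative effects.
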Material /Game/Environments/Artemis/Sets/Daisy/GiantTree_A/Materials/M_DaisyHeroTree_Master
graph {
 properties {
  bUseMaterialAttributes=true
  bUsedWithInstancedStaticMeshes=true
  bUsedWithNanite=true
  blendMode=BLEND_Masked
  ditheredLODTransition=true
  pixelDepthOffsetMode=PDOM_Legacy
  shadingModel=MSM_Subsurface
  usedShadingModels=Subsurface
 }
 nodes [
  n1 = RuntimeVirtualTextureSampleParameter!(UVs, World Position, Mip Level=Constant() {r=8}) {
   materialType=BaseColor_Normal_Specular_YCoCg
   mipValueMode=RVTMVM_MipLevel
   parameterName=RVTColor
   virtualTexture=/Game/Athena/Apollo/Maps/Landscape/VT/RVT_Apollo_Landscape
  }
  n2 = TextureCoordinate() {coordinateIndex=1}
  n3 = StaticSwitchParameter(True=@n2, False=TextureCoordinate()) {parameterName=UseUV2forDetailMaps}
  n4 = Multiply(
   A=ComponentMask(None=VectorParameter() {group=Trunk, parameterName=BarkDetail Tiling, sortPriority=50, defaultValue=<rgba 4 2 0 1>}.RGB) {g=true, r=true}
   B=@n3
  )
  n5 = TextureSampleParameter2D(UVs=@n4) {
   group=Trunk
   parameterName=Trunk_BaseColor
   samplerSource=SSM_Wrap_WorldGroupSettings
   texture=/Game/Global/Textures/Default/Blanks/T_BlackUserInterface
  }
  n6 = LinearInterpolate(
   A=@n5.RGB
   B=VectorParameter() {group=Trunk, parameterName=DetailMaskColor, defaultValue=<rgba 0.0833 0.0833 0.0833 1>}.RGB
   Alpha=NamedRerouteUsage()
  )
  n7 = LinearInterpolate(
   A=Multiply(A=@n6, B=VectorParameter() {group=Trunk, parameterName=Trunk_AO_MultColor, defaultValue=<rgba 0 0 0 1>}.RGB)
   B=@n6
   Alpha=Power(Base=NamedRerouteUsage(), Exp=ScalarParameter() {defaultValue=2, group=Trunk, parameterName=BakedAOPower})
  )
  n8 = LinearInterpolate(
   A=@n7
   B=MaterialFunctionCall(Base=@n7, Blend=NamedRerouteUsage()) {materialFunction=&Blend_Overlay}.Result
   Alpha=ScalarParameter() {defaultValue=1, group=Trunk, parameterName=CurvatureInfluence}
  )
  n9 = Subtract(A=WorldPosition().XYZ, B=Add(A=ActorPositionWS(), B=ScalarParameter() {group=Trunk, parameterName=Ground Blending Offset}))
  n10 = ScalarParameter() {defaultValue=256, group=Trunk, parameterName=Ground Blending Fade Out Range}
  n11 = Saturate(None=Divide(A=ComponentMask(None=@n9) {b=true}, B=Multiply(A=@n10) {constB=2, desc=why}))
  n12 = StaticSwitchParameter(
   True=LinearInterpolate(A=VectorParameter() {parameterName=GroundBlendingColor, defaultValue=<rgba 0.1302 0.0805 0.0274 1>}.RGB, B=@n8, Alpha=@n11)
   False=@n8
  ) {defaultValue=true, parameterName=UseGroundBlend}
  n13 = Multiply(
   A=ComponentMask(None=VectorParameter() {group=Trunk, parameterName=MicroDetail Tiling, sortPriority=50, defaultValue=<rgba 4 2 0 1>}.RGB) {g=true, r=true}
   B=@n3
  )
  n14 = TextureSampleParameter2D(UVs=@n13) {
   group=Trunk
   parameterName=Trunk_DetailMap
   samplerSource=SSM_Wrap_WorldGroupSettings
   samplerType=SAMPLERTYPE_Masks
   texture=/Game/Environments/Apollo/Foliage/Textures_Materials/Apollo_Bark_Detail_D_01
  }
  n15 = ComponentMask(None=@n14.RGB) {r=true}
  n16 = Multiply(A=ScalarParameter() {defaultValue=1.0272, parameterName=Detail Strength}, B=Subtract(A=Multiply(A=@n15) {constB=2}))
  n17 = Multiply(A=@n12, B=Add(A=@n16))
  n18 = QualitySwitch(Default=QualitySwitch(Default=@n12, High=@n12, Epic=Saturate(None=@n17)), Low=@n8, Medium=@n8)
  n19 = Saturate(None=Multiply(A=@n16) {constB=-1})
  n20 = TextureSampleParameter2D(UVs=@n4) {
   group=Trunk
   parameterName=Trunk_Specular
   samplerSource=SSM_Wrap_WorldGroupSettings
   samplerType=SAMPLERTYPE_Masks
   texture=&T_Fortnite_Default_S
  }
  n21 = QualitySwitch(Default=QualitySwitch(Default=@n20.R, High=@n20.R, Epic=Subtract(A=@n20.R, B=@n19)), Low=@n20.R, Medium=@n20.R)
  n22 = Constant() {r=1}
  n23 = QualitySwitch(
   Default=StaticSwitchParameter(True=LinearInterpolate(A=@n22, B=@n20.B, Alpha=@n11), False=@n20.B) {defaultValue=true, parameterName=UseGroundBlend}
   Low=@n20.B
   Medium=@n20.B
  )
  n24 = StaticSwitchParameter(True=Constant() {r=1}, False=Constant() {desc=So nobody forgets to set it to opaque}) {parameterName=UseOpaqueCanopy}
  n25 = TextureSampleParameter2D() {
   group=Trunk
   parameterName=Baked_Trunk_Normal
   samplerSource=SSM_Wrap_WorldGroupSettings
   samplerType=SAMPLERTYPE_Normal
   texture=/Game/Environments/Apollo/Foliage/Textures_Materials/T_Oak_Leaves_N
  }
  n26 = TextureSampleParameter2D(UVs=@n4) {
   group=Trunk
   parameterName=Trunk_Normal
   samplerSource=SSM_Wrap_WorldGroupSettings
   samplerType=SAMPLERTYPE_Normal
   texture=/Game/Environments/Apollo/Foliage/Textures_Materials/Apollo_Bark_Detail_N_01
  }
  n27 = Transform(None=MaterialFunctionCall!(UVs=@n2, Normal=@n26.RGB) {materialFunction=/Engine/Functions/Engine_MaterialFunctions01/Texturing/DeriveTangentBasis}.World Space Normal) {transformSourceType=TRANSFORMSOURCE_World, transformType=TRANSFORM_Tangent}
  n28 = MaterialFunctionCall(
   BaseNormal=@n25.RGB
   AdditionalNormal=MaterialFunctionCall(
    Normal=StaticSwitchParameter(True=@n27, False=@n26.RGB) {parameterName=UseUV2forDetailMaps}
    Flatness=ScalarParameter() {defaultValue=0.5, parameterName=BarkNormalFlatten}
   ) {materialFunction=&FlattenNormal}.Result
  ) {materialFunction=&BlendAngleCorrectedNormals}
  n29 = TextureSampleParameter2D(UVs=@n13) {
   group=Trunk
   parameterName=Trunk_Micro_Normal
   samplerSource=SSM_Wrap_WorldGroupSettings
   samplerType=SAMPLERTYPE_Normal
   texture=&Apollo_Bark_Detail_N_01
  }
  n30 = StaticSwitchParameter(
   True=Transform(None=MaterialFunctionCall(UVs=@n2, Normal=@n29.RGB) {materialFunction=&DeriveTangentBasis}.World Space Normal) {transformSourceType=TRANSFORMSOURCE_World, transformType=TRANSFORM_Tangent}
   False=@n29.RGB
  ) {parameterName=UseUV2forDetailMaps}
  n31 = QualitySwitch(
   Default=@n30
   Epic=MaterialFunctionCall(Normal=@n30, Flatness=ScalarParameter() {defaultValue=-2, parameterName=BarkDetailFlattnenorma}) {materialFunction=&FlattenNormal}.Result
  )
  n32 = QualitySwitch(
   Default=QualitySwitch(
    Default=@n28.Result
    High=@n28.Result
    Epic=MaterialFunctionCall(BaseNormal=@n28.Result, AdditionalNormal=@n31) {materialFunction=&BlendAngleCorrectedNormals}.Result
   )
   Low=@n25.RGB
   Medium=@n25.RGB
  )
  n33 = Constant2Vector()
  n34 = WorldPosition() {worldPositionShaderOffset=WPT_ExcludeAllShaderOffsets}
  n35 = RuntimeVirtualTextureSampleParameter!(UVs, World Position) {materialType=WorldHeight, parameterName=RVTHeight, virtualTexture=/Game/Athena/Apollo/Maps/Landscape/VT/RVT_Apollo_Landscape_Height}
  n36 = Min(A=Subtract(A=ComponentMask(None=@n34.XYZ) {b=true}, B=@n35.WorldHeight)) {constB=200}
  n37 = ComponentMask(None=MaterialFunctionCall() {materialFunction=&LocalPosition}."Local Position (Excluding Offsets)") {b=true}
  n38 = Subtract(A=@n34.XYZ, B=Add(A=ActorPositionWS(), B=ScalarParameter() {defaultValue=50, group=Trunk, parameterName=Ground Snapping Offset}))
  n39 = ScalarParameter() {defaultValue=500, group=Trunk, parameterName=Ground Snapping Fade Out Range}
  n40 = Saturate(None=Divide(A=ComponentMask(None=@n38) {b=true}, B=Multiply(A=@n39) {constB=2}))
  n41 = Constant() {r=1}
  n42 = ScalarParameter() {group=Trunk, parameterName=Ground Snapping Fade Out Constrast}
  n43 = Multiply(
   A=Multiply(A=Subtract(A=@n36, B=@n37)) {constB=-1}
   B=OneMinus(None=MaterialFunctionCall(Transition Phase=@n40, Height Texture=@n41, Contrast=@n42) {materialFunction=&HeightLerp}.Alpha)
  )
  n44 = ActorPositionWS()
  n45 = Subtract(
   A=ComponentMask(None=@n44) {g=true, r=true}
   B=ComponentMask(None=Constant3Vector() {constant=<rgba -5602.9888 -5602.9399 0 0>}) {g=true, r=true}
  ) {constB=2.9888}
  n46 = Constant() {r=0.0088}
  n47 = TextureSample(UVs=Multiply(A=Multiply(A=@n45, B=@n46) {constB=0.0088}) {constB=0.01}) {
   samplerSource=SSM_Wrap_WorldGroupSettings
   samplerType=SAMPLERTYPE_LinearGrayscale
   texture=/Game/Environments/Apollo/Foliage/Textures_Materials/Apollo_Test_Heightmap_VS
  }
  n48 = Multiply(A=@n47.R) {constB=51200}
  n49 = Abs(None=Subtract(A=ComponentMask(None=@n44) {b=true}, B=Subtract(A=@n48) {constB=25600}))
  n50 = Floor(None=Saturate(None=Multiply(A=@n49) {constB=0.01}))
  n51 = StaticSwitchParameter(True=Multiply(A=AppendVector(A=@n33, B=@n43), B=OneMinus(None=@n50)), False=Constant()) {group=Trunk, parameterName=Use Landscape Base Snapping}
  n52 = StaticSwitchParameter(True=Constant3Vector(), False=Constant3Vector() {constant=<rgba 8 0.95 0 0>}) {parameterName=UseOpaqueCanopy}
  n53 = WorldPosition()
  n54 = Sine(None=Divide(A=ComponentMask(None=@n53.XYZ) {b=true}) {constB=50})
  n55 = TextureCoordinate()
  n56 = Constant() {r=1}
  n57 = MaterialFunctionCall!(
   In=MaterialFunctionCall!(
    In=MakeMaterialAttributes!(
     BaseColor=@n18
     Metallic
     Specular=@n21
     Roughness=@n23
     Anisotropy
     EmissiveColor
     Opacity=@n24
     OpacityMask=@n54
     Normal=@n32
     Tangent
     WorldPositionOffset=@n51
     SubsurfaceColor=@n52
     ClearCoat
     ClearCoatRoughness
     AmbientOcclusion
     Refraction
     CustomizedUVs_0
     CustomizedUVs_1
     CustomizedUVs_2
     CustomizedUVs_3
     CustomizedUVs_4
     CustomizedUVs_5
     CustomizedUVs_6
     CustomizedUVs_7
     PixelDepthOffset
     ShadingModel
     Displacement
    )
    FlipTwoSidedSign
    IsTangentSpaceInput=StaticBool() {value=true}
    HasBackFlipFaceNormals
   ) {materialFunction=/Game/Packages/Fortnite_UniqueMaterials_1/MaterialFunctions/Trees/MF_Tree_Snow}.Result
   Mask=If(A=ComponentMask(None=@n55) {r=true}, B=Constant() {r=0.995}, "A > B"=Constant(), "A == B"=@n56, "A < B"=@n56)
  ) {desc="Used for Opaque Perf optmization ", materialFunction=/Game/Environments/Apollo/Foliage/MaterialFunctions/MF_DistanceWPOfade}
  n58 = TwoSidedSign()
  n59 = Constant() {r=1}
  n60 = StaticSwitch(
   True=MaterialFunctionCall!(
    Material Attributes=@n57.Result
    IsTangentSpaceNormal=StaticBool() {desc=Don't change, value=true}
    In=StaticSwitchParameter(True=@n58, False=@n59) {defaultValue=true, parameterName=UseFlipBackFaceNormals}
   ) {materialFunction=/Game/ImpostorBaker/MaterialFunctions/ImposterCaptureSwitch}.Result
   False=@n57.Result
   Value=StaticBool()
  )
  n61 = GetMaterialAttributes(None=@n60) {attributeGetTypes [#1, #2]}
  n62 = TextureSampleParameter2D() {
   group=Emissive
   parameterName=EmissiveTexture
   samplerSource=SSM_Wrap_WorldGroupSettings
   samplerType=SAMPLERTYPE_Masks
   texture=/Engine/MaterialTemplates/LayerBlends/T_DefaultMaskTexture
  }
  n63 = ChannelMaskParameter(None=@n62.RGBA) {maskChannel=Blue, parameterName=EmissiveMask, defaultValue=<rgba 0 0 1 0>}
  n64 = WorldPosition() {worldPositionShaderOffset=WPT_ExcludeAllShaderOffsets}
  n65 = CameraVectorWS()
  n66 = Add(
   A=TransformPosition(None=VectorParameter() {group=Emissive, parameterName=EmissiveCenterPosition, defaultValue=<rgba 0 0 0 1>}.RGB) {transformType=TRANSFORMPOSSOURCE_World}
   B=Multiply(A=@n65, B=ScalarParameter() {group=Emissive, parameterName=EmissiveCenterSoften})
  )
  n67 = PixelNormalWS()
  n68 = ScalarParameter() {parameterName=EmissivePixelNormalDistort}
  n69 = Normalize(VectorInput=Add(A=Subtract(A=@n64.XYZ, B=@n66), B=Multiply(A=@n67, B=@n68)))
  n70 = Multiply(
   A=Power(Base=DotProduct(A=@n69, B=@n65), Exp=ScalarParameter() {defaultValue=3, group=Emissive, parameterName=EmissiveFresnel_InterirorFalloff})
   B=VectorParameter() {group=Emissive, parameterName=EmissiveFresnel_CoreBrightness, defaultValue=<rgba 2 1.1601 0 1>}.RGB
  )
  n71 = LinearInterpolate(
   A=ScalarParameter() {group=Emissive, parameterName=EmissiveContrast_1}
   B=ScalarParameter() {defaultValue=1, group=Emissive, parameterName=EmissiveContrast_2}
   Alpha=ChannelMaskParameter(None=StaticSwitchParameter(True=@n62.RGBA, False=@n61.Base Color) {parameterName=UseEmissiveTexture}) {maskChannel=Blue, parameterName=EmissiveMask, defaultValue=<rgba 0 0 1 0>}.R
  )
  n72 = ScalarParameter() {defaultValue=1, group=Emissive, parameterName=EmissivePower}
  n73 = StaticSwitchParameter(
   True=StaticSwitchParameter(True=Multiply(A=@n63.R, B=@n70), False=@n70) {parameterName=UseEmissiveTexture}
   False=Multiply(A=VectorParameter() {group=Emissive, parameterName=EmissiveColor, defaultValue=<rgba 1 1 1 1>}.RGB, B=Power(Base=@n71, Exp=@n72))
  ) {parameterName=UseEmissiveInterirorFresnel}
  n74 = ShadingPathSwitch(Default=@n73, Mobile=Clamp(None=@n73, Max=ScalarParameter() {defaultValue=5, parameterName=MobileMaxBrightness})) {desc=Migrated from FeatureLevelSwitch}
  n75 = StaticSwitchParameter(
   True=MaterialFunctionCall!(In=SetMaterialAttributes(MaterialAttributes=@n61.MaterialAttributes, Emissive Color=@n74) {attributeSetTypes [#2]}) {materialFunction=/Game/Athena/WinterArt/Materials/MF_TOD_EmissiveColors}.Result
   False=@n60
  ) {parameterName=UseEmissive}
  n76 = MaterialFunctionCall!(
   TrunkHeight=NamedRerouteUsage()
   Material=MaterialFunctionCall!(
    AO=NamedRerouteUsage()
    Material=MaterialFunctionCall!(In=@n75) {materialFunction=/Game/Environments/Props/Winter/Materials/EmissiveTextureWinterTree}.Material Output
   ) {materialFunction=/Game/Environments/Props/Winter/Materials/MF_FrontierTreeEmissive}.Result
  ) {materialFunction=&/MF_MossCoverage}
  n77 = GetMaterialAttributes(None=@n76.Result) {attributeGetTypes [#1]}
  n78 = LinearInterpolate(
   A=@n77.Base Color
   B=Multiply(A=@n77.Base Color, B=VectorParameter() {group=VertexAO, parameterName=VertexAOColor, defaultValue=<rgba 0 0 0 1>}.RGB)
   Alpha=OneMinus(None=VertexColor().G)
  )
  n79 = GetMaterialAttributes(None=StaticSwitchParameter(
    True=SetMaterialAttributes!(MaterialAttributes=@n77.MaterialAttributes, Base Color=@n78) {attributeSetTypes [#1]}
    False=@n76.Result
   ) {group=VertexAO, parameterName=UseVertexBakedAO}) {attributeGetTypes [#3]}
  n80 = SetMaterialAttributes!(
   MaterialAttributes=@n79.MaterialAttributes
   World Position Offset=Add(
    A=@n79.World Position Offset
    B=MaterialFunctionCall!(
     WSDeformationCenter
     Code_DestructionAmount
     Code_UseUvForDamage
     Code_UseHitEffects
     Code_Emissive
     WSDamageRadius
     CrackDamageShader
     Diffuse
     ScalarModDamageCracks
     ScalarOffsetforDamageCracks
     ExposeLerpFunctionForInvertingTheHitDistanceCalc
     InvertDamageFalloff
     CrackWorldPosition
     WorldPosition
     DistanceCullFadeAnimationOffset
    ) {materialFunction=/Game/Packages/jl_tests/Fortnite_HitEffects}.Hit WPO
   )
  ) {attributeSetTypes [#3]}
  n81 = StaticSwitchParameter(True=@n80, False=StaticSwitchParameter(True=@n80, False=@n79.MaterialAttributes) {group=Trunk, parameterName=IsTrunk}) {group=Trunk, parameterName=ForceEnable_HitFXWobble}
  n82 = ShadingPathSwitch(Default=@n81, Deferred=@n81, Forward=@n79.MaterialAttributes, Mobile=@n79.MaterialAttributes)
  n83 = VertexColor()
  n84 = MaterialFunctionCall(X=@n83.G, Y=@n83.G, Z=@n83.G) {materialFunction=&MakeFloat3}
  MaterialExpressionAppendVector_2 = AppendVector(A=Add(A=Multiply(A=@n83.R) {constB=2}), B=Add(A=Multiply(A=@n84.Result) {constB=2}) {constB=-1})
  n85 = CollectionParameter() {
   collection=&FortniteMaterialParameters
   parameterName=SunlightDirectionalVector
  }
  n86 = ComponentMask(None=@n85) {b=true, g=true, r=true}
  MaterialExpressionTransform_0 = Transform(None=@n86) {transformSourceType=TRANSFORMSOURCE_World, transformType=TRANSFORM_Local}
  MaterialExpressionConstant_1 = Constant() {r=1}
  MaterialExpressionConstant_2 = Constant()
  n87 = Custom!(
   SH=@MaterialExpressionAppendVector_2
   Direction=@MaterialExpressionTransform_0
   Scale=@MaterialExpressionConstant_1
   Offset=@MaterialExpressionConstant_2
  ) {
   code="//SH decode,  Direction = LightDir, SH = 2band coefficients. (encoded in Y-up)\r\nDirection = normalize(Direction);\r\nfloat4 shEvalDirection;\r\nshEvalDirection.x = 0.282095;\r\nshEvalDirection.y =0.488603 * Direction.z;\r\nshEvalDirection.z = 0.488603 * Direction.y;\r\nshEvalDirection.w =-0.488603 * Direction.x;\r\nreturn saturate(dot(SH, shEvalDirection));"
   description=SphericalHarmonicsEval
   outputType=CMOT_Float1
   inputs [
    {
     inputName=SH
     input {expression=&self:MaterialExpressionAppendVector_2}
    }
    {
     inputName=Direction
     input {expression=&self:MaterialExpressionTransform_0}
    }
    {
     inputName=Scale
     input {expression=&self:MaterialExpressionConstant_1}
    }
    {
     inputName=Offset
     input {expression=&self:MaterialExpressionConstant_2}
    }
   ]
  }
  n88 = OneMinus(None=@n87)
  n89 = Power(Base=@n88, Exp=ScalarParameter() {defaultValue=2, group=ShadowGradient, parameterName=ShadowGradientContrast, sliderMax=5, sliderMin=0.5})
  n90 = ScalarParameter() {defaultValue=2, group=ShadowGradient, parameterName=ShadowGradientIntensity, sliderMax=5, sliderMin=0.5}
  n91 = Normalize(VectorInput=Subtract(A=WorldPosition().XYZ, B=ObjectPositionWS()))
  n92 = VertexNormalWS()
  n93 = Saturate(None=DotProduct(A=@n86, B=StaticSwitchParameter(True=@n91, False=@n92) {group=ShadowGradient, parameterName=ColorGradientWorldSpace}))
  n94 = ShadingPathSwitch(Default=Saturate(None=Multiply(A=@n89, B=@n90)), Mobile=@n93) {desc=Migrated from FeatureLevelSwitch}
  n95 = MaterialFunctionCall!(In=MaterialFunctionCall!(In=@n82, Mask=VertexInterpolator(VS=@n94).PS, IsImpostor) {materialFunction=/Game/Environments/Apollo/Foliage/MaterialFunctions/MF_TreeFresnel}.Result) {materialFunction=/Game/Environments/Apollo/Foliage/MaterialFunctions/MF_SubSurfaceFadeEpicDetail}
  n96 = StaticSwitchParameter(True=StaticSwitchParameter(True=@n95.Result, False=@n82) {parameterName=UseOpaqueCanopy}, False=@n95.Result) {group=Trunk, parameterName=IsTrunk}
  n97 = GetMaterialAttributes(None=@n96) {attributeGetTypes [#2, #1]}
  n98 = VectorParameter() {group=GlowSphere, parameterName=GlowSphereColor, sortPriority=1}
  n99 = LinearInterpolate(A=Constant() {r=1}, B=@n97.Base Color, Alpha=ScalarParameter() {group=GlowSphere, parameterName=GlowDiffuseModToEmissive})
  n100 = WorldPosition()
  n101 = Constant()
  n102 = Add(
   A=ActorPositionWS()
   B=AppendVector(A=AppendVector(A=@n101, B=@n101), B=ScalarParameter() {group=GlowSphere, parameterName=GlowSphereLocationOffsetZ, sortPriority=4})
  )
  n103 = ScalarParameter() {group=GlowSphere, parameterName=GlowSphereRadius, sortPriority=2}
  n104 = ScalarParameter() {desc=Go negative to have it softer, group=GlowSphere, parameterName=GlowSphereContrast, sortPriority=3}
  n105 = CollectionParameter() {
   collection=&FortniteMaterialParameters
   parameterName=FortniteActiveTimeOfDay
  }
  n106 = VectorParameter() {
   group=GlowSphere
   parameterName=GlowSpereTOD
   sortPriority=7
   channelNames {a=Night, b=Evening, g=Day, r=Morning}
   defaultValue=<rgba 0.5 0.2 0.5 1>
  }
  n107 = Multiply(
   A=Multiply(A=Multiply(A=@n98.RGB, B=@n99), B=SphereMask(A=@n100.XYZ, B=@n102, Radius=@n103, Hardness=@n104))
   B=DotProduct(A=@n105, B=AppendVector(A=@n106, B=@n106))
  )
  n108 = MaterialProxyReplace(
   Realtime=SetMaterialAttributes!(MaterialAttributes=@n97.MaterialAttributes, Emissive Color=Add(A=@n97.Emissive Color, B=@n107)) {attributeSetTypes [#2]}
   MaterialProxy=@n97.MaterialAttributes
  )
  n109 = MaterialFunctionCall!(
   MA In=StaticSwitchParameter(True=QualitySwitch(Default=@n108, Low=@n97.MaterialAttributes, High=@n108, Medium=@n97.MaterialAttributes, Epic=@n108), False=@n96) {group=GlowSphere, parameterName=GlowSphere, sortPriority=-3}
   SnowMaskOverride
   Fresnel_SnowBlendExponent_Override
   Fresnel_SnowBlendBrightness_Override
  ) {materialFunction=/Game/Packages/Fortnite_UniqueMaterials_1/MaterialFunctions/Trees/MF_Tree_Snow_s19}
  n110 = SetMaterialAttributes!(MaterialAttributes=@n109.Result, Base Color=@n1.BaseColor, Specular=@n1.BaseColor, Roughness=@n1.BaseColor) {attributeSetTypes [#1, #4, #5]}
  n111 = RuntimeVirtualTextureSampleParameter() {materialType=WorldHeight, parameterName=RVTHeight, virtualTexture=&RVT_Apollo_Landscape_Height}
  n112 = WorldPosition()
  n113 = ObjectRadius()
  n114 = VertexNormalWS()
  n115 = LinearInterpolate(
   A=ScalarParameter() {defaultValue=0.3, group=RVT, parameterName=Blend RVT North}
   B=ScalarParameter() {defaultValue=0.2, group=RVT, parameterName=Blend RVT}
   Alpha=Saturate(None=ComponentMask(None=@n114) {r=true})
  )
  n116 = Saturate(None=SphereMask(
    A=@n111.WorldHeight
    B=ComponentMask(None=@n112.XYZ) {b=true}
    Radius=Multiply(A=@n113, B=@n115)
    Hardness=ScalarParameter() {defaultValue=0.2, group=RVT, parameterName=BlendRVTContrast}
   ))
  n117 = MaterialProxyReplace(
   Realtime=MaterialFunctionCall(NoRVT=@n109.Result, RVT=BlendMaterialAttributes(A=@n109.Result, B=@n110, Alpha=@n116) {vertexAttributeBlendType=UseA}) {materialFunction=&FN_RVT_Switch_Material}.Result
   MaterialProxy=@n109.Result
  )
  n118 = StaticSwitchParameter(True=@n109.Result, False=QualitySwitch(Default=@n117, Low=@n109.Result, High=@n109.Result, Medium=@n109.Result, Epic=@n117)) {group=RVT, parameterName=Disable_RVT}
  MaterialFunctionCall(MA In=MaterialFunctionCall!(
    In=StaticSwitchParameter(True=@n118, False=@n109.Result) {group=Trunk, parameterName=IsTrunk}
    UseSphericalMappingOverride=StaticBool() {value=true}
   ) {materialFunction=/Game/Packages/Fortress_SharedMaterials/Base_Material_Functions/MF_CurieFallbackEffect}.Result) {materialFunction=&MF_CreativeMoveTool}
  NamedRerouteDeclaration(None=@n5.A) {name=BarkHeight, nodeColor=<rgba 1 0.3516 0 1>}
  n119 = TextureSampleParameter2D() {group=Trunk, parameterName=TrunkMask, samplerType=SAMPLERTYPE_Masks, texture=&T_Fortnite_Default_S}
  NamedRerouteDeclaration(None=@n119.G) {name=BakedAO, nodeColor=<rgba 0 0.9531 1 1>}
  NamedRerouteDeclaration(None=@n119.B) {name=DetailMask, nodeColor=<rgba 0 0.9062 1 1>}
  NamedRerouteDeclaration(None=@n119.R) {name=BakedCurvature, nodeColor=<rgba 0.2891 1 0 1>}
  n120 = TextureSampleParameter2D(UVs=@n4) {
   parameterName=TrunkDetailEmissiveMask
   samplerType=SAMPLERTYPE_Masks
   texture=&Apollo_Bark_Detail_D_01
  }
  NamedRerouteDeclaration(None=@n120.RGB) {name=DetailEmissiveMask, nodeColor=<rgba 0.5938 0 1 1>}
  NamedRerouteUsage()
  NamedRerouteUsage()
  n121 = MaterialFunctionCall(Float3=ActorPositionWS()) {materialFunction=&BreakOutFloat3Components}
  n122 = ScalarParameter() {defaultValue=13000, group=Color, parameterName=Color Variation Distance WS, sortPriority=10}
  n123 = ScalarParameter() {defaultValue=11000, group=Color, parameterName=Color Variation Distance WS2, sortPriority=10}
  n124 = Add(A=Sine(None=Divide(A=@n121.R, B=@n122)), B=Sine(None=Divide(A=@n121.G, B=@n123)))
  VertexInterpolator(VS=Divide(A=Add(A=@n124) {constB=2}) {constB=4})
  n125 = Constant() {r=0.5}
  VertexInterpolator(VS=Saturate(None=Subtract(A=@n93, B=@n125)))
  VertexInterpolator(VS=@n93)
  n126 = Subtract(A=@n93, B=Constant() {r=-0.5})
  VertexInterpolator(VS=OneMinus(None=Saturate(None=@n126)))
  n127 = ScalarParameter() {defaultValue=4, group=ShadowGradient, parameterName=ShadowGradientSpecularContrast, sliderMax=5, sliderMin=0.5}
  VertexInterpolator(VS=LinearInterpolate(
    A=ScalarParameter() {defaultValue=0.25, group=Shading Leafs, parameterName=Specular_Leafs_Shadow}
    B=ScalarParameter() {defaultValue=0.25, group=Shading Leafs, parameterName=Specular_Leafs}
    Alpha=Power(Base=@n88, Exp=@n127)
   ))
  VertexInterpolator()
  VertexInterpolator(VS=Divide(A=@n124))
 ]
}
