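# Master two-sided, masked subsurface, Nanite-ready material for the Daisy giant tree canopy leaves.
# Base color remaps diffuse green through color-atlas curves, then layers clump, canopy-gradient, sun-facing and fresnel fuzz shading.
# Vertex-color-offset sines drive optional wind displacement; UseWeepyWind adds masked macro gusts and skips sun-facing and fuzz shading.
# Optional weepy emissive uses a panning pulse with cheaper quality fallbacks, and MaterialProxyReplace simplifies proxy bakes.
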Material /Game/Environments/Artemis/Sets/Daisy/GiantTree_A/Materials/M_DaisyTree_Canopy_Master
graph {
 properties {
  bUseMaterialAttributes=true
  bUsedWithNanite=true
  blendMode=BLEND_Masked
  pixelDepthOffsetMode=PDOM_Legacy
  shadingModel=MSM_Subsurface
  twoSided=true
  usedShadingModels=Subsurface
 }
 nodes [
  n1 = Multiply(A=Time(), B=ScalarParameter() {defaultValue=512, parameterName=MacroWindSpeed})
  n2 = NamedRerouteUsage()
  n3 = Divide(
   A=Add(A=WorldPosition().XYZ, B=Multiply(A=@n1, B=@n2))
   B=VectorParameter() {parameterName=MacroWindScale, defaultValue=<rgba 2048 2048 2048 1>}.RGB
  )
  ChannelMaskParameter(None=@n3) {parameterName=MacroWindDirection}
  n4 = VertexColor()
  Clamp(None=@n4.R) {minDefault=0.2}
  CollectionParameter() {collection=&FortniteMaterialParameters, parameterName=WindStrength}
  CollectionParameter() {collection=&FortniteMaterialParameters, parameterName=WindSpeed}
  Constant3Vector() {constant=<rgba 0 0 1 0>}
  Constant3Vector() {constant=<rgba -16750.6426 -84665.2969 3991.7627 0>}
  n5 = CollectionParameter() {
   collection=&FortniteMaterialParameters
   parameterName=SunlightDirectionalVector
  }
  n6 = DotProduct(A=ComponentMask(None=@n5) {b=true, g=true, r=true}, B=PixelNormalWS())
  ConstantBiasScale(None=@n6)
  Constant()
  Constant() {r=1}
  n7 = SkyLightEnvMapSample(Direction=ReflectionVectorWS(), Roughness=ScalarParameter() {parameterName=SkylightEnvironemntRoughness})
  n8 = CurveAtlasRowParameter(CurveTime=NamedRerouteUsage()) {
   atlas=/Game/Athena/Apollo/Environments/Landscape/Materials/Frontier/FrontierColorAtlas
   curve=/Game/Athena/Apollo/Environments/Landscape/Materials/Frontier/FrontierCanopyColorVariationCurve
   defaultValue=3
   parameterName=CanopyColorVariation
  }
  n9 = TextureSampleParameter2D() {parameterName=Diffuse, texture=/Game/Environments/Artemis/Sets/Daisy/GiantTree_A/Textures/T_Daisy_GiantTree_Leaves01_D}
  n10 = CurveAtlasRowParameter(CurveTime=@n9.G) {
   atlas=&FrontierColorAtlas
   curve=/Game/Athena/Apollo/Environments/Landscape/Materials/Frontier/FrontierTreeCanopyColorCurve
   defaultValue=2
   parameterName=CanopyColorCurve
  }
  n11 = CurveAtlasRowParameter(CurveTime=@n9.G) {
   atlas=&FrontierColorAtlas
   curve=/Game/Athena/Apollo/Environments/Landscape/Materials/Frontier/FrontierTreeCanopyColorCurve2
   defaultValue=4
   parameterName=CanopyColorCurve2
  }
  n12 = NamedRerouteUsage()
  n13 = LinearInterpolate(
   A=Multiply(A=@n8, B=@n10)
   B=@n11
   Alpha=Power(Base=OneMinus(None=@n12), Exp=ScalarParameter() {defaultValue=2, parameterName=ColorGradientPower}) {constExponent=1.5}
  )
  n14 = LinearInterpolate(
   A=Multiply(A=@n13, B=ScalarParameter() {defaultValue=0.5, parameterName=LocalClumpDarkening})
   B=Multiply(A=@n13, B=ScalarParameter() {defaultValue=2, parameterName=LocalClumpBrightness})
   Alpha=Power(Base=@n12) {constExponent=4}
  )
  n15 = LinearInterpolate(
   A=Multiply(A=@n14, B=ScalarParameter() {defaultValue=0.25, parameterName=CanopyGradientDarkening})
   B=Multiply(A=@n14, B=ScalarParameter() {defaultValue=1.5, parameterName=CanopyGradientBrightening})
   Alpha=NamedRerouteUsage()
  )
  n16 = LinearInterpolate(
   A=Multiply(A=ScalarParameter() {defaultValue=1, parameterName=DarkenSunFacing}, B=@n15)
   B=Multiply(A=@n15, B=ScalarParameter() {defaultValue=2, parameterName=BrightenSunFacing})
   Alpha=NamedRerouteUsage()
  )
  n17 = ScalarParameter() {defaultValue=0.35, parameterName=FuzzDarkness}
  n18 = ScalarParameter() {defaultValue=1.5, parameterName=FuzzBrightness}
  n19 = StaticSwitchParameter(True=@n15, False=LinearInterpolate(A=Multiply(A=@n16, B=@n17), B=Multiply(A=@n16, B=@n18), Alpha=NamedRerouteUsage())) {parameterName=UseWeepyWind}
  LinearInterpolate(A=@n19, B=@n7, Alpha=ScalarParameter() {defaultValue=0.15, parameterName=SS_SkylightColorBlend})
  LinearInterpolate()
  n20 = NamedRerouteUsage()
  n21 = Multiply(A=@n7, B=ScalarParameter() {defaultValue=0.02, parameterName=SkylightColorInfluence})
  n22 = TextureSampleParameter2D() {
   parameterName=FXMask
   samplerType=SAMPLERTYPE_Masks
   texture=/Game/Environments/Artemis/Sets/Daisy/GiantTree_A/Textures/T_Daisy_GiantTree_Leaves01_MSK
  }
  n23 = TextureSampleParameter2D() {
   parameterName=Normal
   samplerType=SAMPLERTYPE_Normal
   texture=/Game/Environments/Artemis/Sets/Daisy/GiantTree_A/Textures/T_Daisy_GiantTree_CanopyDome_N
  }
  n24 = ScalarParameter() {defaultValue=0.02, parameterName=MaxInscatterPower}
  n25 = ScalarParameter() {parameterName=MinInscatterPower}
  n26 = NamedRerouteUsage()
  n27 = ScalarParameter() {defaultValue=0.05, parameterName=MaxSurfaceSCatterPower}
  n28 = ScalarParameter() {parameterName=MinSurfaceSCatterPower}
  MakeMaterialAttributes!(
   BaseColor=MaterialProxyReplace(Realtime=@n19, MaterialProxy=NamedRerouteUsage())
   Metallic=Constant()
   Specular=Constant() {r=1}
   Roughness=Constant() {r=0.9}
   Anisotropy
   EmissiveColor=StaticSwitchParameter(True=Add(A=@n20, B=@n21), False=@n21) {parameterName=UseWeepyEmissive}
   Opacity=LinearInterpolate(A=LinearInterpolate(A=@n24, B=@n25, Alpha=@n26), B=LinearInterpolate(A=@n27, B=@n28, Alpha=@n26), Alpha=NamedRerouteUsage()) {constA=0.25, constB=0}
   OpacityMask=ChannelMaskParameter(None=@n22.RGB) {parameterName=OpacityChannel}.R
   Normal=@n23.RGB
   Tangent
   WorldPositionOffset=NamedRerouteUsage()
   SubsurfaceColor=NamedRerouteUsage()
   ClearCoat
   ClearCoatRoughness
   AmbientOcclusion
   Refraction
   CustomizedUVs_0
   CustomizedUVs_1
   CustomizedUVs_2
   CustomizedUVs_3
   CustomizedUVs_4
   CustomizedUVs_5
   CustomizedUVs_6
   CustomizedUVs_7
   PixelDepthOffset
   ShadingModel
   Displacement
  )
  n29 = WorldPosition()
  n30 = VectorParameter() {parameterName=WSCanopyPivot, defaultValue=<rgba -12744.8447 -79649.7656 4120 1>}
  n31 = Abs(None=Subtract(A=VectorParameter() {parameterName=MaxHeight, defaultValue=<rgba -7602.8154 -74584.2188 9652.3936 1>}.RGB, B=@n30.RGB))
  n32 = ChannelMaskParameter(None=Divide(A=Subtract(A=@n29.XYZ, B=@n30.RGB), B=@n31)) {maskChannel=Blue, parameterName=WSGradient, defaultValue=<rgba 0 0 1 0>}
  MaterialProxyReplace(Realtime=@n32.R, MaterialProxy=Constant() {r=1})
  n33 = DotProduct(A=CameraVectorWS(), B=PixelNormalWS())
  n34 = Clamp(None=Power(Base=OneMinus(None=@n33), Exp=ScalarParameter() {defaultValue=4, parameterName=FresnelPower}))
  MaterialProxyReplace(Realtime=@n34, MaterialProxy=Constant() {r=0.5})
  n35 = Clamp(None=DotProduct(A=Constant3Vector() {constant=<rgba 0 0 -1 0>}, B=VertexNormalWS()))
  MaterialProxyReplace(Realtime=@n35, MaterialProxy=Constant() {r=0.5})
  n36 = Clamp(None=@n6)
  MaterialProxyReplace(Realtime=@n36, MaterialProxy=Constant() {r=0.5})
  Multiply(
   A=VectorParameter() {parameterName=SkylightTIng, defaultValue=<rgba 0.1156 0.5282 1 1>}.RGB
   B=ComponentMask(None=MaterialFunctionCall() {materialFunction=/Engine/Functions/GetPostProcessFunctions/GetAmbientCubemapTint}.AmbientCubemapTint) {b=true, g=true, r=true}
  )
  n37 = NamedRerouteUsage()
  Multiply(A=@n37, B=@n4.G)
  Multiply(
   A=@n26
   B=LinearInterpolate(A=ScalarParameter() {parameterName=MinSS}, B=ScalarParameter() {defaultValue=1, parameterName=MaxSS}, Alpha=NamedRerouteUsage()) {constB=0.5}
  )
  NamedRerouteDeclaration(None=@n34) {name=FresnelFuzzMask, nodeColor=<rgba 0.7344 1 0 1>}
  NamedRerouteDeclaration(None=@n36) {name=LightDirectionMask, nodeColor=<rgba 1 0 0.7031 1>}
  NamedRerouteDeclaration(None=@n35) {name=ZDotN, nodeColor=<rgba 0.8047 0 1 1>}
  n38 = VectorParameter() {parameterName=SSTInt, defaultValue=<rgba 1 0.8245 0.733 1>}
  n39 = ScalarParameter() {defaultValue=2, parameterName=SS_Intensity}
  n40 = ScalarParameter() {defaultValue=1, parameterName=SS_Intensity_Min}
  n41 = NamedRerouteUsage()
  NamedRerouteDeclaration(None=Multiply(A=Multiply(A=@n19, B=@n38.RGB), B=LinearInterpolate(A=@n39, B=@n40, Alpha=@n41))) {name=SSColor, nodeColor=<rgba 0.3125 1 0 1>}
  NamedRerouteDeclaration(None=Add(A=NamedRerouteUsage(), B=@n13)) {name=MaterialProxyColors, nodeColor=<rgba 0.3359 0 1 1>}
  n42 = PreSkinnedPosition()
  n43 = VertexInterpolator(VS=@n42)
  n44 = Divide(
   A=Subtract(
    A=ShaderStageSwitch(PixelShader=@n43.PS, VertexShader=@n42)
    B=VectorParameter() {parameterName=CanopyChunkPivotOIffset, defaultValue=<rgba 0 0 0 1>}.RGB
   )
   B=MaterialFunctionCall() {materialFunction=&ObjectLocalBounds}.Local Bounds Size
  )
  NamedRerouteDeclaration(None=MaterialProxyReplace(
    Realtime=ChannelMaskParameter(None=@n44) {maskChannel=Blue, parameterName=ZGradientChannel, defaultValue=<rgba 0 0 1 0>}.R
    MaterialProxy=Constant() {r=0.5}
   )) {name=Z-Gradient, nodeColor=<rgba 0 0.3906 1 1>}
  n45 = CollectionParameter() {
   collection=&FortniteMaterialParameters
   parameterName=WindDirection_RGB Speed_A
  }
  n46 = Normalize(VectorInput=Multiply(A=ComponentMask(None=@n45) {b=true, g=true, r=true}) {constB=-1})
  n47 = QualitySwitch(Default=@n46)
  NamedRerouteDeclaration(None=ShadingPathSwitch(Default=@n47, Forward=@n47, Mobile=@n47) {desc=Need this to prevent compile error}) {name=WindDirection, nodeColor=<rgba 0 1 0.5547 1>}
  n48 = Time()
  n49 = ScalarParameter() {parameterName=Min_TimeOffset}
  n50 = ScalarParameter() {defaultValue=1, parameterName=Max_TimeOffset}
  n51 = VertexColor()
  n52 = Add(A=@n48, B=LinearInterpolate(A=@n49, B=@n50, Alpha=@n51.R))
  n53 = Multiply(
   A=CollectionParameter() {collection=&FortniteMaterialParameters, parameterName=WindSpeed}
   B=LinearInterpolate(
    A=ScalarParameter() {defaultValue=0.25, parameterName=Min_SpeedOffset}
    B=ScalarParameter() {defaultValue=0.5, parameterName=Max_SpeedOffset}
    Alpha=@n51.G
   )
  )
  n54 = ScalarParameter() {defaultValue=0.5, parameterName=Min_StrengthOffset}
  n55 = ScalarParameter() {defaultValue=1, parameterName=Max_StrengthOffset}
  n56 = ChannelMaskParameter(None=TextureCoordinate()) {maskChannel=Green, parameterName=StripGradientChannel, defaultValue=<rgba 0 1 0 0>}
  n57 = Multiply(
   A=Multiply(A=ScalarParameter() {defaultValue=32, parameterName=Debug_WindStrength}, B=LinearInterpolate(A=@n54, B=@n55, Alpha=@n51.R))
   B=@n56.R
  )
  n58 = Multiply(A=Sine(None=Multiply(A=@n52, B=@n53)), B=@n57)
  n59 = Sine(None=@n3)
  n60 = ScalarParameter() {defaultValue=512, parameterName=WindStrength}
  n61 = MaterialFunctionCall(Float2=Multiply(A=ConstantBiasScale(None=@n59), B=Multiply(A=@n56.R, B=@n60))) {materialFunction=&BreakOutFloat2Components}
  n62 = MaterialFunctionCall(X=@n61.R, Y=@n61.R, Z=Constant()) {materialFunction=&MakeFloat3}
  n63 = Add(A=@n48, B=LinearInterpolate(A=@n49, B=@n50, Alpha=@n51.G))
  n64 = Multiply(A=Sine(None=Multiply(A=@n63, B=@n53)), B=@n57)
  n65 = Add(A=@n48, B=LinearInterpolate(A=@n49, B=@n50, Alpha=@n51.G))
  n66 = Multiply(A=Sine(None=Multiply(A=@n65, B=@n53)), B=@n57)
  n67 = MaterialFunctionCall(
   X=LinearInterpolate(A=@n58, B=Add(A=@n58, B=@n62.Result), Alpha=@n51.B)
   Y=LinearInterpolate(A=@n64, B=Add(A=@n64, B=@n62.Result), Alpha=@n51.B)
   Z=@n66
  ) {materialFunction=&MakeFloat3}
  n68 = MaterialFunctionCall(X=@n58, Y=@n64, Z=@n66) {materialFunction=&MakeFloat3}
  NamedRerouteDeclaration(None=StaticSwitchParameter(True=StaticSwitchParameter(True=@n67.Result, False=@n68.Result) {parameterName=UseWeepyWind}, False=Constant()) {parameterName=UseWindAnimation}) {name=WindWPO, nodeColor=<rgba 1 0.4219 0 1>}
  NamedRerouteDeclaration(None=@n22.G) {name=ColorVariationMask, nodeColor=<rgba 0.5469 0 1 1>}
  n69 = ComponentMask(None=TextureCoordinate()) {g=true}
  n70 = Subtract(A=@n69, B=ScalarParameter() {parameterName=EmissiveGradientOffset})
  n71 = NamedRerouteUsage()
  n72 = Multiply(
   A=Multiply(A=Clamp(None=@n70), B=@n4.B)
   B=LinearInterpolate(
    A=VectorParameter() {parameterName=EmissiveColor2, defaultValue=<rgba 0 0.2253 0.3281 1>}.RGB
    B=VectorParameter() {parameterName=EmissiveColor, defaultValue=<rgba 0 0.6928 1 1>}.RGB
    Alpha=@n71
   )
  )
  n73 = ScalarParameter() {defaultValue=1, parameterName=EmissiveStrength}
  n74 = LinearInterpolate(A=@n73, B=Multiply(A=@n73) {constB=4}, Alpha=@n71)
  n75 = Add(
   A=Time()
   B=LinearInterpolate(
    A=ScalarParameter() {parameterName=EmissiveTime_MinOffset}
    B=ScalarParameter() {defaultValue=1, parameterName=EmissiveTime_MaxOffset}
    Alpha=Multiply(A=@n37, B=@n4.R)
   )
  )
  n76 = LinearInterpolate(
   A=ScalarParameter() {defaultValue=0.1, parameterName=Emissive_MinSpeed}
   B=ScalarParameter() {defaultValue=0.25, parameterName=Emissive_MaxSpeed}
   Alpha=@n4.G
  )
  n77 = Sine(None=Subtract(A=@n69, B=Multiply(A=@n75, B=@n76)))
  n78 = ScalarParameter() {parameterName=EmissivePanSample}
  n79 = ScalarParameter() {defaultValue=0.25, parameterName=EmissivePanRadius}
  n80 = ScalarParameter() {parameterName=EmissivePanHardness}
  n81 = Multiply(A=@n72, B=LinearInterpolate(A=@n74, B=Multiply(A=@n74) {constB=2}, Alpha=SphereMask(A=@n77, B=@n78, Radius=@n79, Hardness=@n80)))
  n82 = Multiply(A=@n72, B=@n73)
  n83 = QualitySwitch(Default=@n81, Low=@n82, Epic=@n81)
  n84 = QualitySwitch(Default=@n82, Low=@n82, Medium=@n82)
  NamedRerouteDeclaration(None=Power(Base=ShadingPathSwitch(Default=@n83, Mobile=@n84))) {name=Emissive, nodeColor=<rgba 1 0 0.8672 1>}
  NamedRerouteDeclaration(None=@n32.R) {name=GlobalCanopyZGradeint, nodeColor=<rgba 0 0.2969 1 1>}
  NamedRerouteDeclaration(None=@n9.G) {name=DiffuseRef, nodeColor=<rgba 0 1 0.0625 1>}
  NamedRerouteDeclaration(None=@n31) {name=CanopyBounds, nodeColor=<rgba 1 0 0.4453 1>}
  NamedRerouteUsage()
  NamedRerouteUsage()
  NamedRerouteUsage()
  Normalize(VectorInput=Subtract(A=WorldPosition().XYZ, B=ObjectPositionWS()))
  Normalize(VectorInput=Add(A=@n46))
  n85 = CollectionParameter() {
   collection=&FortniteMaterialParameters
   parameterName=SunlightDirectionalVector
  }
  OneMinus(None=@n85)
  n86 = CollectionParameter() {
   collection=&FortniteMaterialParameters
   parameterName=SunlightDirectionalVector
  }
  OneMinus(None=ComponentMask(None=@n86) {b=true, g=true, r=true})
  PerInstanceRandom()
  PixelNormalWS()
  Power(Base=@n36) {constExponent=12}
  n87 = CollectionParameter() {
   collection=&FortniteMaterialParameters
   parameterName=WeatherWindDirectionAndStrength
  }
  n88 = Saturate(None=Add(
    A=MaterialFunctionCall() {materialFunction=/Game/Athena/Environments/Landscape/MaterialFunctions/MF_StormMask}.Result
    B=ComponentMask(None=@n87) {a=true}
   ))
  n89 = Add(B=@n88)
  ShadingPathSwitch(Default=QualitySwitch(Default=@n88, Epic=Saturate(None=@n89)), Forward=@n88, Mobile=@n88) {desc=Need this to prevent compile error}
  Subtract(A=NamedRerouteUsage(), B=Frac(None=Time()))
  VectorParameter() {parameterName=MAcroWindSpeed, defaultValue=<rgba 512 512 512 1>}
  VectorParameter() {parameterName=MinHeight, defaultValue=<rgba -16750.6426 -84665.2969 3991.7627 1>}
  VectorParameter() {parameterName=GradientColor, defaultValue=<rgba 0.1562 0.325 1 1>}
  VectorParameter() {parameterName=WSCanopyBounds, defaultValue=<rgba 0 0 0 1>}
  VectorParameter() {parameterName=GradientColor2, defaultValue=<rgba 0.1538 0.0102 0.599 1>}
  VertexColor()
  VertexNormalWS()
 ]
}
